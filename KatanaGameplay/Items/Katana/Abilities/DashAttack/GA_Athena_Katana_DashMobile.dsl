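; GA_Athena_Katana_DashMobile is a placeholder gameplay ability in the Katana weapon's secondary ability slot, existing so the secondary button appears on mobile.
; It ends itself immediately on activation, because staying active makes FortWeapon treat the weapon as firing and block actions like swim sprint.
; Mantis consumes and handles the secondary input, so the ability has no gameplay effect.
; The failed passive activation event only calls the parent.
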(event Ability|EventActivateAbility
  ;[1 0.9359 0.09279] This GA is a placeholder for the secondary ability in the Katana's WID to get the secondary button to show up on mobile. It doesn't actually do anything as the secondary input is being consumed and handled by Mantis.
  ;
  ; It must be ended as soon as it's activated because not ending it will result in the FortWeapon class thinking this weapon is firing and block actions like swim sprint. Mantis should be in control of those actions at all times, so canceling this ability will allow it to get out of the way. Ending this ability does not have any adverse effects on the mobile behavior.
  (|CallParentFunction)
  (Ability|EndAbility))

(event Ability|EventFailedToActivatePassiveAbility
  (params (ActorInfo :type Gameplay%20Ability%20Actor%20Info%20Structure))
  ;[1 0.9359 0.09279] This GA is a placeholder for the secondary ability in the Katana's WID to get the secondary button to show up on mobile. It doesn't actually do anything as the secondary input is being consumed and handled by Mantis.
  ;
  ; It must be ended as soon as it's activated because not ending it will result in the FortWeapon class thinking this weapon is firing and block actions like swim sprint. Mantis should be in control of those actions at all times, so canceling this ability will allow it to get out of the way. Ending this ability does not have any adverse effects on the mobile behavior.
  (|CallParentFunction ActorInfo))
