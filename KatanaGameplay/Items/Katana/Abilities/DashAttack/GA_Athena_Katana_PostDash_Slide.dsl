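; Gameplay ability for the Katana's post-dash slide, applying a predicted movement effect after a dash attack.
; A timer or gaining the MovementMode.Sliding tag triggers StopPostDashSlide.
; Stopping adds loose suppress tags on client and server, working around unpredictable effect removal, while authority removes the grant and movement effects.
; High-acceleration net tolerance is enabled on activation, and ability end reverts it and clears tags, timer and task.
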(event Ability|EventActivateAbility
  (|CallParentFunction)
  ;[0.09279 1 0.4823] Apply from the GA to get a client prediction key.
  (Ability|ApplyGameplayEffectToOwner "/KatanaGameplay/Items/Katana/Abilities/DashAttack/GE_Katana_PostDash_SlideMovement.GE_Katana_PostDash_SlideMovement_C")
  (bind _returnvalue (Utilities|Time|SetTimerbyEvent (EventDispatchers|CreateEvent :!event Custom|StopPostDashSlide) (Variables|Default|GetPostDashSlideDuration)))
  (Variables|Default|SetStopPostDashSlideTimerHandle _returnvalue)
  (bind _outfortplayerpawn (CallFunction|GetAvatarasValidFortPlayerPawn)
    (:ValidFortPlayerPawn
      (bind _asfort_movement_comp_character (Utilities|Casting|CastToFortMovementComp_Character (Variables|/Script/Engine.Character|GetCharacterMovement _outfortplayerpawn)))
      (NetTolerance|SetUseHighAccelerationTolerance _asfort_movement_comp_character true "Katana_PostDashSlide")
      (bind _asynctaskproxy (Ability|Tasks|WaitGameplayTagAdd "(TagName=\"MovementMode.Sliding\")" 0 true)
        (:then
          (Variables|Default|SetWaitSlidingAsyncTask _asynctaskproxy))
        (:Added
          (CallFunction|StopPostDashSlide))))))

(event Ability|EventFailedToActivatePassiveAbility
  (params (ActorInfo :type Gameplay%20Ability%20Actor%20Info%20Structure))
  (|CallParentFunction ActorInfo))

(event Custom|StopPostDashSlide
  (bind _postdashslidesuppresstags (Variables|Default|GetPostDashSlideSuppressTags))
  (Utilities|Time|ClearandInvalidateTimerbyHandle (Variables|Default|GetStopPostDashSlideTimerHandle))
  ;[1 0.3037 0.8964] Ignore the Movement GE on the server and client. This makes it so the GE removal can be replicated since there's no way to predictively remove GEs.
  (bind _outfortplayerpawn (CallFunction|GetAvatarasValidFortPlayerPawn)
    (:ValidFortPlayerPawn
      (if (not (GameplayTags|HasAnyMatchingGameplayTags _outfortplayerpawn _postdashslidesuppresstags))
        (Ability|GameplayEffect|AddLooseGameplayTags _outfortplayerpawn _postdashslidesuppresstags))
      (Ability|HasAuthority
        (:True
          ;[0 1 0.7754] Removes the Grant GE and the Movement GE.
          (Ability|RemoveGameplayEffectFromOwnerWithGrantedTags (Variables|Default|GetPostDashSlideTags)))))))

(event Ability|EventOnEndAbility
  (params (bWasCancelled :type Boolean))
  (bind _postdashslidesuppresstags (Variables|Default|GetPostDashSlideSuppressTags))
  (|CallParentFunction bWasCancelled)
  (Utilities|Time|ClearandInvalidateTimerbyHandle (Variables|Default|GetStopPostDashSlideTimerHandle))
  (bind _outfortplayerpawn (CallFunction|GetAvatarasValidFortPlayerPawn)
    (:ValidFortPlayerPawn
      (if (GameplayTags|HasAnyMatchingGameplayTags _outfortplayerpawn _postdashslidesuppresstags)
        (Ability|GameplayEffect|RemoveLooseGameplayTags _outfortplayerpawn _postdashslidesuppresstags)
        (else
          _))
      (bind _asfort_movement_comp_character (Utilities|Casting|CastToFortMovementComp_Character (Variables|/Script/Engine.Character|GetCharacterMovement _outfortplayerpawn)))
      (NetTolerance|SetUseHighAccelerationTolerance _asfort_movement_comp_character false "Katana_PostDashSlide")
      (bind _waitslidingasynctask (Variables|Default|GetWaitSlidingAsyncTask)
        (:then
          (GameplayTasks|EndTask _waitslidingasynctask))
        (:else)))))
