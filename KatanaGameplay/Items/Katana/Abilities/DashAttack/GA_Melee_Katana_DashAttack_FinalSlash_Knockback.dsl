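; Knockback ability for the Katana dash attack's final slash, overriding how the launched target's velocity is computed.
; CalculateVelocityMaintained casts the instigator to a Fort player pawn and takes that pawn's PawnForward targeting direction.
; It scales that direction by a data-driven XY magnitude and adds a data-driven Z lift, ignoring the incoming launch vector.
; If the cast fails it defers to the parent calculation, and activation events only call the parent.
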(event Ability|EventActivateAbility
  (|CallParentFunction))

(event Ability|EventFailedToActivatePassiveAbility
  (params (ActorInfo :type Gameplay%20Ability%20Actor%20Info%20Structure))
  (|CallParentFunction ActorInfo))

(fn CalculateVelocityMaintained
  (params (LaunchTarget :type Actor%20Object%20Reference) (LaunchMagIn :type Vector :default "0, 0, 0"))
  (bind _asfort_player_pawn (Utilities|Casting|CastToFortPlayerPawn (Default|GetAbilityInstigator))
    (:then
      (bind (_returnvalue_location _returnvalue_rotation) (Math|Transform|BreakTransform (Targeting|GetTargetingTransform _asfort_player_pawn "PawnForward" 0)))
      (return (Utilities|Operators|Add (* (Math|Vector|GetForwardVector _returnvalue_rotation) (ScalableFloat|GetValueatLevel (Variables|KnockbackParent|GetRow_XYKnockMag))) (Math|Vector|MakeVector 0.0 0.0 (ScalableFloat|GetValueatLevel (Variables|KnockbackParent|GetRow_ZAdd))))))
    (:CastFailed
      (return (|CallParentFunction (Variables|GetLaunchTarget) (Variables|GetLaunchMagIn))))))
